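# Editor utility widget that renders a Lowpoly turntable sequence through Movie Render Queue from a Render button.
# A Details View bound to the widget itself exposes its properties, such as preset and output directory.
# RenderSequence clears the queue, creates a job with map, sequence, MRQPresets config and output directory, then runs the PIE executor.
# Soft-reference preset lookup and anti-aliasing override branches are unwired; GenerateCustomSequence_Morph is a stub.
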EditorUtilityWidgetBlueprint /Lowpoly_Turntable/Lowpoly_MRQ_Controls
graphs [
 EventGraph {nodes [
   "On Clicked (Button_102)"() {is=title K2Node_ComponentBoundEvent}
   > RenderSequence()
   Event Construct()
   > SetObject(self=Get DetailsView_112() -> (DetailsView_112:Details View Object Reference), NewObject=Self-Reference())
   Event PreConstruct()
   Event Tick()
  ]}
 GenerateCustomSequence_Morph {nodes [
   GenerateCustomSequence_Morph() {is=title K2Node_FunctionEntry}
   > Return Node!(execute, HasValidTargets:bool=false)
  ]}
 RenderSequence {nodes [
   n1 = TT_Func_GetAssetIfValid!(
    execute
    SoftObjectReference:"Soft Object Reference (by ref)"=Get FN_SettingPresets() -> (FN_SettingPresets:Soft Object Reference)
    __WorldContext:Object Reference
   ) -> (then, Is Valid?:bool, ReturnValue:Object Reference)
   > n2 = Branch(Condition=@n1)
   > n3 = Cast To MoviePipelinePrimaryConfig!(execute, Object=@n1.ReturnValue) -> (then, CastFailed, AsMovie Pipeline Primary Config:Movie Pipeline Primary Config Object Reference) {is=title K2Node_DynamicCast}
   n4 = Get MoviePipelineQueueSubsystem() -> (ReturnValue:Movie Pipeline Queue Subsystem Object Reference) {is=title K2Node_GetEditorSubsystem}
   n5 = GetQueue!(self:Movie Pipeline Queue Subsystem Object Reference=@n4) -> (ReturnValue:Movie Pipeline Queue Object Reference)
   n6 = AllocateNewJob!(
    execute=@n11.then_1
    self:Movie Pipeline Queue Object Reference=@n5
    InJobType:Movie Pipeline Executor Job Class Reference=/Script/MovieRenderPipelineCore.MoviePipelineExecutorJob
   ) -> (then, ReturnValue:Movie Pipeline Executor Job Object Reference)
   n7 = SetConfiguration!(execute=@n3.then, self:Movie Pipeline Executor Job Object Reference=@n6, InPreset:Movie Pipeline Primary Config Object Reference=@n3) -> (then)
   n8 = GetConfiguration!(self:Movie Pipeline Executor Job Object Reference=@n6) -> (ReturnValue:Movie Pipeline Primary Config Object Reference)
   n9 = FindOrAddSettingByClass!(
    execute=[@n10.then, @n7.then, @n13.Completed]
    self:Movie Pipeline Config Base Object Reference=@n8
    InClass:Movie Pipeline Setting Class Reference=/Script/MovieRenderPipelineCore.MoviePipelineOutputSetting
    bIncludeDisabledSettings:bool=false
    bExactMatch:bool=false
   ) -> (then, ReturnValue:Movie Pipeline Output Setting Object Reference)
   > Set OutputDirectory!(
    execute
    OutputDirectory:Directory Path Structure=Get Output Directory() -> (Output Directory:Directory Path Structure)
    self:Movie Pipeline Output Setting Object Reference=@n9
   ) -> (then, Output_Get:Directory Path Structure)
   n10 = SetConfiguration(
    execute=[@n3.CastFailed, @n2.else, @n18.then]
    self=@n6
    InPreset=Get MRQPresets() -> (MRQPresets:Movie Pipeline Primary Config Object Reference)
   )
   RenderSequence() {is=title K2Node_FunctionEntry}
   > n11 = Sequence!(execute) -> (then_0, then_1, then_2:Exec)
   > n12 = For Each Loop(Array:Array of Movie Pipeline Executor Job Object References=GetJobs!(self:Movie Pipeline Queue Object Reference=@n5) -> (ReturnValue:Array of Movie Pipeline Executor Job Object References)) -> (Array Element:Movie Pipeline Executor Job Object Reference)
   > DeleteJob!(execute, self:Movie Pipeline Queue Object Reference=@n5, InJob:Movie Pipeline Executor Job Object Reference=@n12) -> (then)
   RenderQueueWithExecutor!(
    execute=@n11.then_2
    self:Movie Pipeline Queue Subsystem Object Reference=@n4
    InExecutorType:Movie Pipeline Executor Base Class Reference=/Script/MovieRenderPipelineEditor.MoviePipelinePIEExecutor
   ) -> (then, ReturnValue:Movie Pipeline PIEExecutor Object Reference)
   n13 = For Each Loop(Array:Array of Movie Pipeline Setting Object References=GetAllSettings!(self:Movie Pipeline Primary Config Object Reference=@n8, bIncludeDisabledSettings:bool=false, bIncludeTransientSettings:bool=false) -> (ReturnValue:Array of Movie Pipeline Setting Object References)) -> (Array Element:Movie Pipeline Setting Object Reference)
   > n14 = Cast To MoviePipelineAntiAliasingSetting!(execute, Object=@n13) -> (then, CastFailed, AsMovie Pipeline Anti Aliasing Setting:Movie Pipeline Anti Aliasing Setting Object Reference) {is=title K2Node_DynamicCast}
   > n15 = Switch on ECheckBoxState!(
    execute
    Selection:ECheckBoxState Enum=GetCheckedState!(self:Check Box Object Reference=Get EnableTAA_Checkbox() -> (EnableTAA_Checkbox:Check Box Object Reference)) -> (ReturnValue:ECheckBoxState Enum=Unchecked)
   ) -> (Unchecked:Exec, Checked:Exec, Undetermined:Exec) {is=title K2Node_SwitchEnum}
   Set bOverrideAntiAliasing!(execute=[@n15.Checked, @n15.Undetermined], bOverrideAntiAliasing:bool=false, self:Movie Pipeline Anti Aliasing Setting Object Reference=@n14) -> (then, Output_Get:bool)
   Set bOverrideAntiAliasing(execute=@n15.Unchecked, bOverrideAntiAliasing=true, self=@n14)
   > Set AntiAliasingMethod!(execute, AntiAliasingMethod:EAntiAliasingMethod Enum=AAM_None, self:Movie Pipeline Anti Aliasing Setting Object Reference=@n14) -> (then, Output_Get:EAntiAliasingMethod Enum=AAM_None)
   > PrintString(InString="~~~~~~~~~~~~~~~~~ Overriding AntiAliasing to None")
   GetCurrentShotFrameNumber!(InMoviePipeline:Movie Pipeline Object Reference) -> (ReturnValue_Value:int)
   PrintString(InString=GetDisplayName(Object=@n12))
   n16 = Get Map() -> (Map:Soft Object Path Structure)
   n17 = Get MRQ_Sequence() -> (MRQ_Sequence:Soft Object Path Structure)
   n18 = Set Map!(execute=@n19.then, Map:Soft Object Path Structure=@n16, self:Movie Pipeline Executor Job Object Reference=@n6) -> (then, Output_Get:Soft Object Path Structure)
   Set JobName!(execute=@n6.then, JobName:String=Lowpoly_Turntable_Render, self:Movie Pipeline Executor Job Object Reference=@n6) -> (then, Output_Get:String)
   > n19 = Set Sequence!(execute, Sequence:Soft Object Path Structure=@n17, self:Movie Pipeline Executor Job Object Reference=@n6) -> (then, Output_Get:Soft Object Path Structure)
  ]}
]
variables [
 MRQPresets {type=MoviePipelinePrimaryConfig, default=/Lowpoly_Turntable/Lowpoly_TT_Assets/Lowpoly_MRQ_Settings_Default}
 Output Directory {type=DirectoryPath, default {path=C:/DEFAULTDONTUSE}}
]
widgets {
 info {rootWidgetClass=&Overlay, widgetCount=5}
 widgets [
  Overlay_31 {widgetClassPath=&Overlay, widget=Overlay_31}
  VerticalBox_58 {widgetClassPath=&VerticalBox, parent=Overlay_31, widget=VerticalBox_58, slot {class=OverlaySlot}}
  Button_102 {
   widgetClassPath=&Button
   parent=VerticalBox_58
   widget=Button_102
   properties {widgetStyle {
     clickedSlateSound {resourceObject=null}
     disabled {
      bIsDynamicallyLoaded=false
      drawAs=RoundedBox
      imageType=NoImage
      mirroring=NoMirror
      resourceName=None
      resourceObject=null
      tiling=NoTile
      imageSize=<0 0>
      margin=<ltrb 0 0 0 0>
      outlineSettings {
       bUseBrushTransparency=false
       roundingType=FixedRadius
       width=1
       color {colorUseRule=UseColor_ColorTable, specifiedColor=<rgba 0 0 0 0>}
       cornerRadii=<4 4 4 4>
      }
      tintColor {colorUseRule=UseColor_Specified, specifiedColor=<rgba 0.0395 0.0395 0.0395 1>}
      uVRegion {bIsValid=false, max=<0 0>, min=<0 0>}
     }
     disabledForeground {colorUseRule=UseColor_ColorTable, specifiedColor=<rgba 0 0 0 0>}
     hovered {
      bIsDynamicallyLoaded=false
      drawAs=RoundedBox
      imageType=NoImage
      mirroring=NoMirror
      resourceName=None
      resourceObject=null
      tiling=NoTile
      imageSize=<32 32>
      margin=<ltrb 0 0 0 0>
      outlineSettings {
       bUseBrushTransparency=false
       roundingType=FixedRadius
       width=1
       color {colorUseRule=UseColor_ColorTable, specifiedColor=<rgba 0 0 0 0>}
       cornerRadii=<4 4 4 4>
      }
      tintColor {colorUseRule=UseColor_Specified, specifiedColor=<rgba 0.0953 0.0953 0.0953 1>}
      uVRegion {bIsValid=false, max=<0 0>, min=<0 0>}
     }
     hoveredForeground {colorUseRule=UseColor_ColorTable, specifiedColor=<rgba 0 0 0 0>}
     hoveredSlateSound {resourceObject=null}
     normal {
      bIsDynamicallyLoaded=false
      drawAs=RoundedBox
      imageType=NoImage
      mirroring=NoMirror
      resourceName=None
      resourceObject=null
      tiling=NoTile
      imageSize=<32 32>
      margin=<ltrb 0 0 0 0>
      outlineSettings {
       bUseBrushTransparency=false
       roundingType=FixedRadius
       width=1
       color {colorUseRule=UseColor_ColorTable, specifiedColor=<rgba 0 0 0 0>}
       cornerRadii=<4 4 4 4>
      }
      tintColor {colorUseRule=UseColor_Specified, specifiedColor=<rgba 0.0395 0.0395 0.0395 1>}
      uVRegion {bIsValid=false, max=<0 0>, min=<0 0>}
     }
     normalForeground {colorUseRule=UseColor_ColorTable, specifiedColor=<rgba 0 0 0 0>}
     normalPadding=<ltrb 12 1.5 12 1.5>
     pressed {
      bIsDynamicallyLoaded=false
      drawAs=RoundedBox
      imageType=NoImage
      mirroring=NoMirror
      resourceName=None
      resourceObject=null
      tiling=NoTile
      imageSize=<32 32>
      margin=<ltrb 0 0 0 0>
      outlineSettings {
       bUseBrushTransparency=false
       roundingType=FixedRadius
       width=1
       color {colorUseRule=UseColor_ColorTable, specifiedColor=<rgba 0 0 0 0>}
       cornerRadii=<4 4 4 4>
      }
      tintColor {colorUseRule=UseColor_Specified, specifiedColor=<rgba 0.0284 0.0284 0.0284 1>}
      uVRegion {bIsValid=false, max=<0 0>, min=<0 0>}
     }
     pressedForeground {colorUseRule=UseColor_ColorTable, specifiedColor=<rgba 0 0 0 0>}
     pressedPadding=<ltrb 12 2.5 12 0.5>
     pressedSlateSound {resourceObject=null}
    }}
   slot {class=VerticalBoxSlot}
  }
  TextBlock_108 {widgetClassPath=&TextBlock, parent=Button_102, widget=TextBlock_108, properties {text=Render}, slot {class=ButtonSlot}}
  DetailsView_112 {
   widgetClassPath=&DetailsView
   parent=VerticalBox_58
   widget=DetailsView_112
   properties {categoriesToShow [NoCategory]}
   slot {class=VerticalBoxSlot}
  }
 ]
}
